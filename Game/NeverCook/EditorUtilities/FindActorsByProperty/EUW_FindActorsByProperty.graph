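# Editor utility widget that finds actors by property in a chosen world through a TechArt Python library and lists the returned paths.
# A details view bound to the widget exposes WorldToSearch; Find Actors times the search.
# Load Last Search Data reloads prior results via Python, and ValidateList rebuilds the list from either result.
# Shift-click syncs the Content Browser to an entry; Save List As Collection creates or empties a collection, then fills it.
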EditorUtilityWidgetBlueprint /Game/NeverCook/EditorUtilities/FindActorsByProperty/EUW_FindActorsByProperty
graphs [
 EventGraph {nodes [
   Delay!(execute, WorldContextObject:Object Reference, Duration=1) -> (then)
   > LoadLastActors()
   "On Clicked (TA_Generic01_Button)"!() -> (OutputDelegate, then, widget:User Widget Object Reference, event:Pointer Event Structure) {is=title K2Node_ComponentBoundEvent}
   n1 = Break Soft World Reference!(SoftWorldReference:"Soft World Reference Structure (by ref)"=Get WorldToSearch() -> (WorldToSearch:Soft World Reference Structure)) -> (WorldAsset:World Soft Object Reference) {is=title K2Node_BreakStruct}
   > Branch(Condition=IsValidSoftObjectReference(SoftObjectReference=@n1))
   > n2 = GetPlatformTime_Seconds!(execute) -> (then, ReturnValue:double)
   > FindActorsWithPropertyInMap()
   > n3 = GetPlatformTime_Seconds()
   n4 = float - float(A=@n3, B=@n2)
   > PrintString(InString=Append(A="FindActorsByProperty Elapsed Time: ", B="To String (Float)"(InDouble=@n4)))
   n5 = "On Item Left Clicked (TA_Generic01_List)"!() -> (OutputDelegate, then, isShiftDown:bool, isCtrlDown:bool, data:Object Reference, item:User Widget Object Reference) {is=title K2Node_ComponentBoundEvent}
   > Branch(Condition=@n5.isShiftDown)
   > n6 = Cast To TA_GenericListItemData!(execute, Object=@n5.data) -> (then, CastFailed, AsTA Generic List Item Data:TA Generic List Item Data Object Reference) {is=title K2Node_DynamicCast}
   n7 = Get displayString!(self:TA Generic List Item Data Object Reference=@n6) -> (displayString:String)
   > SyncBrowserToObjects(AssetPaths=Make Array("[0]":String=@n7) -> (Array:Array of Strings))
   "On Clicked (btnLoad)"!() -> (OutputDelegate, then, widget:User Widget Object Reference, event:Pointer Event Structure) {is=title K2Node_ComponentBoundEvent}
   > LoadLastActors()
   "On Clicked (btnCollect)"!() -> (OutputDelegate, then, widget:User Widget Object Reference, event:Pointer Event Structure) {is=title K2Node_ComponentBoundEvent}
   > ExportToCollection()
   Event Construct()
   > SetObject(self=Get DetailsView_150() -> (DetailsView_150:Details View Object Reference), NewObject=Self-Reference())
   PrintString(InString=@n7)
   PrintString(InString="To String (SoftObjectReference)"!(SoftObjectReference:"Soft Object Reference (by ref)"=@n1) -> (ReturnValue:String))
  ]}
 ExportToCollection {nodes [
   ExportToCollection() {is=title K2Node_FunctionEntry}
   > n1 = getListData!(execute, self:TA Generic 01 List Object Reference=Get TA_Generic01_List() -> (TA_Generic01_List:TA Generic 01 List Object Reference)) -> (then, data:Array of Object References)
   > n2 = For Each Loop(Array:Array of Object References=@n1) -> (Array Element:Object Reference)
   n3 = CollectionExists!(
    execute=@n2.Completed
    self:Asset Tags Subsystem Object Reference=Get AssetTagsSubsystem() -> (ReturnValue:Asset Tags Subsystem Object Reference) {is=title K2Node_GetEngineSubsystem}
    Name:Name=Get l_collectionName() -> (l_collectionName:Name=None)
   ) -> (then, ReturnValue:bool)
   > n4 = Branch(Condition=@n3)
   > n5 = EmptyCollection!(execute, self:Asset Tags Subsystem Object Reference=Get AssetTagsSubsystem(), Name:Name=Get l_collectionName() -> (l_collectionName=None)) -> (then, ReturnValue:bool)
   AddAssetsToCollection!(
    execute=[@n6.then, @n5.then]
    self:Asset Tags Subsystem Object Reference=Get AssetTagsSubsystem()
    Name:Name=Get l_collectionName() -> (l_collectionName=None)
    AssetPathNames:Array of Names=Get l_maps() -> (l_maps:Array of Names)
   ) -> (then, ReturnValue:bool)
   n6 = CreateCollection!(
    execute=@n4.else
    self:Asset Tags Subsystem Object Reference=Get AssetTagsSubsystem()
    Name:Name=Get l_collectionName() -> (l_collectionName=None)
    ShareType:ECollectionScriptingShareType Enum=Local
   ) -> (then, ReturnValue:bool)
   n7 = Cast To TA_GenericListItemData(execute=@n2.LoopBody, Object=@n2)
   n8 = Get displayString(self=@n7)
   n9 = Split(SourceString=@n8, InStr=/, SearchDir=FromEnd)
   > Add(
    TargetArray:Array of Names=Get l_maps()
    NewItem:"Name (by ref)"=String To Name(InString=Append!(A=@n8, B=., C:String=@n9.RightS) -> (ReturnValue))
   ) {is=title K2Node_CallArrayFunction}
  ]}
 FindActorsWithPropertyInMap {nodes [
   FindActorsWithPropertyInMap() {is=title K2Node_FunctionEntry}
   n1 = Break Soft World Reference(SoftWorldReference=Get WorldToSearch())
   > n2 = Execute Python Script!(
    execute
    PythonScript="import TechArt.Tools.FindActorsWithProperties.FindActorsWithProperties_Lib as find_actor_lib\r\nimport importlib\r\nimportlib.reload(find_actor_lib)\r\n\r\n##asset_package_name = \"/Game/Athena/Helios/Maps/Helios_Terrain\"\r\n##asset_package_name = \"/Game/Developers/brycelumpkin/Testing/TestingActorDesc/TestingActorDescMap\"\r\n\r\nworlds = find_actor_lib.run_it(asset_package_name)"
    asset_package_name:String=Split(SourceString="To String (SoftObjectReference)"(SoftObjectReference=@n1), InStr=., SearchDir=FromEnd)
   ) -> (then, ReturnValue, worlds:Array of Strings) {inputs [asset_package_name], outputs [worlds]}
   > n3 = Set l_worlds(l_worlds:Array of Strings=@n2.worlds) -> (Output_Get:Array of Strings)
   > ValidateList!(execute, self, Array:Array of Strings=@n3) -> (then)
   MakeLiteralString(Value=&TestingActorDescMap)
  ]}
 LoadLastActors {nodes [
   LoadLastActors() {is=title K2Node_FunctionEntry}
   > n1 = Execute Python Script!(
    execute
    PythonScript="import TechArt.Tools.FindActorsWithProperties.FindActorsWithProperties_Lib as find_actor_lib\r\nimport importlib\r\nimportlib.reload(find_actor_lib)\r\n\r\nworlds = find_actor_lib.loadFindActorsGame()"
   ) -> (then, ReturnValue, worlds) {outputs [worlds]}
   > n2 = Set l_worlds(l_worlds=@n1.worlds)
   > ValidateList(Array=@n2)
  ]}
 ValidateList {nodes [
   n1 = ValidateList!() -> (then, Array:Array of Strings) {is=title K2Node_FunctionEntry}
   > clearItems(self:TA Generic 01 List Object Reference=Get TA_Generic01_List())
   > PrintString(InString="World--------------------------------------------------------\r\n---------------------------------------------------------------")
   > n2 = For Each Loop(Array:Array of Strings=@n1) -> (Array Element:String)
   > PrintString(InString=@n2)
   > n3 = Construct Object from Class(Class=/Game/Blueprints/Blutilities/ForTechArt/Data/TA_GenericListItemData.TA_GenericListItemData_C) -> (ReturnValue:TA Generic List Item Data Object Reference)
   > Set displayString!(execute, displayString:String=@n2, self:TA Generic List Item Data Object Reference=@n3) -> (then, Output_Get:String)
   > addItem!(execute, self:TA Generic 01 List Object Reference=Get TA_Generic01_List(), dataObject:Object Reference=@n3) -> (then, newItem:User Widget Object Reference, newData:Object Reference)
   addItemGeneric!(execute, self:TA Generic 01 List Object Reference=Get TA_Generic01_List(), displayString:String=@n2) -> (then, widget:User Widget Object Reference, data:Object Reference)
  ]}
]
variables [
 allowConstruct {type=boolean, default=true}
 WorldToSearch {type=SoftWorldReference, default {worldAsset=&Helios_Terrain}}
]
widgets {
 info {rootWidgetClass=&StackBox, widgetCount=6}
 widgets [
  StackBox_0 {widgetClassPath=&StackBox, widget=StackBox_0, properties {orientation=Orient_Vertical}}
  DetailsView_150 {
   widgetClassPath=&DetailsView
   parent=StackBox_0
   widget=DetailsView_150
   properties {categoriesToShow [nocategory]}
   slot {class=StackBoxSlot}
  }
  TA_Generic01_Button {
   widgetClassPath=&TA_Generic01_Button
   parent=StackBox_0
   widget=TA_Generic01_Button
   properties {label=Find Actors}
   slot {class=StackBoxSlot}
  }
  btnLoad {
   widgetClassPath=&TA_Generic01_Button
   parent=StackBox_0
   widget=btnLoad
   properties {label=Load Last Search Data}
   slot {class=StackBoxSlot}
  }
  TA_Generic01_List {
   widgetClassPath=/Game/Blueprints/Blutilities/ForTechArt/Widgets/List/TA_Generic01_List
   bIsVariable=true
   parent=StackBox_0
   widget=TA_Generic01_List
   properties {allowMultiSelect=false}
   slot {class=StackBoxSlot, properties {size {sizeRule=Fill, value=1}}}
  }
  btnCollect {
   widgetClassPath=&TA_Generic01_Button
   parent=StackBox_0
   widget=btnCollect
   properties {label=Save List As Collection}
   slot {class=StackBoxSlot}
  }
 ]
}
